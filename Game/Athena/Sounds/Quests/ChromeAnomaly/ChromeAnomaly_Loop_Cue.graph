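# Looping sound cue for the Chrome Anomaly quest, routed to the quest sound class.
# A quality-level node plays the standard looping wave in its first two slots and the "opt" loop in the last two.
# A root modulator adds slight random pitch and volume variation.
# Attenuation reuses the HiddenStar supply-drop ambient loop settings.
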SoundCue /Game/Athena/Sounds/Quests/ChromeAnomaly/ChromeAnomaly_Loop_Cue
root=Modulator_0
nodes [
 {id=Modulator_0, children [QualityLevel_0], properties {pitchMax=1, pitchMin=0.97, volumeMax=1.3, volumeMin=1}}
 {id=QualityLevel_0, children [WavePlayer_1,WavePlayer_1,WavePlayer_0,WavePlayer_0]}
 {id=WavePlayer_0, properties {bLooping=true, soundWaveAssetPtr=&/sfx_chrome_anomaly_opt_loop}}
 {id=WavePlayer_1, properties {bLooping=true, soundWaveAssetPtr=&/sfx_chrome_anomaly_loop}}
]
properties {
 attenuationSettings=/Game/Athena/Sounds/SupplyDrops/Fort_HiddenStar_Ambient_Loop_Att
 maxDistance=2500
 soundClassObject=&Fort_Quests
 volumeMultiplier=1.12
 attenuationOverrides {falloffDistance=700, attenuationShapeExtents {x=100}}
}
